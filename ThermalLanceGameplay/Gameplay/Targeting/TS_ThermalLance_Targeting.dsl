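; Targeting task for the Thermal Lance that shapes a trace from the firing player's pawn and current weapon.
; It overrides source location, direction and length, reading the Fort player pawn from the targeting request's source context.
; The trace starts at the weapon's damage start location, follows the pawn's control rotation forward and spans the weapon's range.
; Each override falls back to the parent implementation when its validity checks fail.
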(fn GetSourceLocation
  (params (TargetingHandle :type Targeting%20Request%20Handle%20Structure%20%28by%20ref%29))
  (bind _returnvalue (Targeting|GetTargetingSubsystem TargetingHandle))
  ; Get the location from which the trace is done.
  (Utilities|FlowControl|Sequence
    (:then_0
      (Utilities|IsValid _returnvalue
        (:Is%20Valid
          (bind _asfort_player_pawn (Utilities|Casting|CastToFortPlayerPawn (Utilities|Struct|BreakTargetingSourceContext (TargetingSystem|TargetingTypes|GetTargetingSourceContext _returnvalue TargetingHandle))))
          (Utilities|IsValid (Weapon|GetCurrentWeapon _asfort_player_pawn)
            (:Is%20Valid
              (return (.ReturnValue (FortWeapon|GetDamageStartLocation (Weapon|GetCurrentWeapon _asfort_player_pawn) -1))))))))
    (:then_1
      (bind _returnvalue_2 (|CallParentFunction TargetingHandle))
      (return _returnvalue_2))))

(fn GetTraceDirection
  (params (TargetingHandle :type Targeting%20Request%20Handle%20Structure%20%28by%20ref%29))
  (bind _returnvalue (Targeting|GetTargetingSubsystem TargetingHandle))
  (Utilities|FlowControl|Sequence
    (:then_0
      (Utilities|IsValid _returnvalue
        (:Is%20Valid
          (bind _asfort_player_pawn (Utilities|Casting|CastToFortPlayerPawn (Utilities|Struct|BreakTargetingSourceContext (TargetingSystem|TargetingTypes|GetTargetingSourceContext _returnvalue TargetingHandle))))
          (return (Math|Vector|GetForwardVector (Pawn|GetControlRotation _asfort_player_pawn :!class /Script/Engine.Pawn))))))
    (:then_1
      (bind _returnvalue_2 (|CallParentFunction TargetingHandle))
      (return _returnvalue_2))))

(fn GetTraceLength
  (params (TargetingHandle :type Targeting%20Request%20Handle%20Structure%20%28by%20ref%29))
  (bind _returnvalue (Targeting|GetTargetingSubsystem TargetingHandle))
  ;[0.03957 0.15 0.1485 0.5] Return the range of the weapon.
  (Utilities|FlowControl|Sequence
    (:then_0
      (Utilities|IsValid _returnvalue
        (:Is%20Valid
          (bind _asfort_player_pawn (Utilities|Casting|CastToFortPlayerPawn (Utilities|Struct|BreakTargetingSourceContext (TargetingSystem|TargetingTypes|GetTargetingSourceContext _returnvalue TargetingHandle))))
          (Utilities|IsValid (Weapon|GetCurrentWeapon _asfort_player_pawn)
            (:Is%20Valid
              (return (FortWeapon|GetRange (Weapon|GetCurrentWeapon _asfort_player_pawn))))))))
    (:then_1
      (bind _returnvalue_2 (|CallParentFunction TargetingHandle))
      (return _returnvalue_2))))
